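# Niagara system for the AlienLiquidPSP vehicle boost cosmetic, emitting sprites and ribbon trails from skeletal mesh sockets, offset by DTCarOffset.
# A GPU Sprites emitter layers spawn groups with SubUV animation, tinted by painted color.
# Trail ribbons shift orange to blue over life, while a parent-template sprite emitter adds curl-noise particles that inherit vehicle motion.
# GameModeID and frontend-preview switches adjust parameters such as spawn rates and velocities.
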NiagaraSystem /VehicleCosmetics/Boosts/Boost_AlienLiquidPSP/FX/NS_Boost_AlienLiquidPSP_Optimized
dependencies {
 usedDataInterfaces [
  &NiagaraDataInterfaceCurve
  &NiagaraDataInterfaceColorCurve
  &NiagaraDataInterfaceSpriteRendererInfo
  &NiagaraDataInterfaceSkeletalMesh
 ]
 usedDynamicInputs []
 usedModules [
  &self:SystemSpawnScript.SystemScriptSource.SystemScriptGraph.NiagaraNodeAssignment_0.SetVariables_#1
  &SystemState
 ]
}
emitters {
 NE_VC_Boost_Sprite_Parent {
  emitterUpdateScript [
   &EmitterState()
   &VC_FilteredSocketCount()
   &SpawnRate(SpawnRate=30, Use Spawn Probability=false)
   &SpawnRate(SpawnGroup=1, SpawnRate=30, Use Spawn Probability=false) {module=SpawnRate001}
  ]
  particleSpawnScript [
   &VC_PaintedVariantsOverride(
    Black=<rgba 1 1 1 1>
    BurntSienna=<rgba 0.5 0.1058 0 1>
    Cobalt=<rgba 0.429 0.4488 1.5 1>
    Crimson=<rgba 1.5 0.3 0.3325 1>
    DefaultColor=<rgba 1.25 0.75 0.25 1>
    ForestGreen=<rgba 0.3464 0.5 0.3541 1>
    Grey=<rgba 0.9 0.9 0.9 0.8>
    Lime=<rgba 0.25 0.5 0 1>
    Orange=<rgba 1 0.335 0.05 1>
    Pink=<rgba 1 0.5 0.8182 1>
    Purple=<rgba 0.6952 0.3438 1.2 1>
    SkyBlue=<rgba 0.2 0.7323 1 1>
   )
   &InitializeParticle(
    Color=<rgba 2 0.6 0.15 1>
    Lifetime=&RandomRangeFloat(Minimum=0.4, Maximum=0.7)
    Sprite Rotation Angle=&RandomRangeFloat(Minimum=0, Maximum=1)
    Sprite Rotation Mode=::NewEnumerator1
    Sprite Size Mode=::NewEnumerator1
    Uniform Sprite Size Max=0.5
    Uniform Sprite Size Min=0.3
    UsePositionOffset=false
    Write Material Random=true
   )
   &PartitionParticles(Number Of Partitions=$Emitter.PortCount)
   &SkeletalMeshLocation(
    Bone / Socket Index=$Particles.PartitionParticles.Partition
    Bone / Socket Sampling Mode=::NewEnumerator5
    MaskSpawnGroup=false
    Mesh Sampling Type=::NewEnumerator6
    Sampled Position Offset=&AddVectorToPosition(Position=<-20 0 0>, Vector=$User.DTCarOffset)
    Skeletal Mesh=$System.SkeletalMesh
   )
   &AddVelocity(Velocity=&RandomRangeVector(Minimum=<-1400 0 10>, Maximum=<-1200 0 -10>), Velocity Speed Scale=0.8)
   /SetVariables(Particles.FakeGlow:Vector2f=<30 30>) {isSetParameters=true}
   &ShapeLocation(MaskSpawnGroup=false, Override Seed=false, Sphere Radius=42, UseOwnerScale=true)
  ]
  particleUpdateScript [
   &ParticleState()
   &CurlNoiseForce(
    Bias Noise=false
    Noise Frequency=50000
    Noise Strength=&RandomRangeFloat(Minimum=500, Maximum=750)
    Pan Noise=false
    Randomization Vector:Vector3f=<0.65 0.125 0>
   )
   &InheritSourceMovement(
    Applied Position Scale=&VectorFromFloat(Value=&FloatFromCurve(FloatCurve=DataInterface() {propertyValues {Curve {keys [{interpMode=RCIM_Cubic, tangentMode=RCTM_Break, time=0.3, value=1}, {interpMode=RCIM_Cubic, tangentMode=RCTM_Break, time=1, value=0}]}}}))
    ApplyForce=false
    ApplyPosition=true
    ApplyPositionFalloff=false
   )
   &ScaleSpriteSize(Scale Sprite Size Mode=::NewEnumerator0, Uniform Scale Factor=15)
   &Drag(
    Drag=&RandomRangeFloat(
     Minimum=0
     Maximum=&FloatFromCurve(FloatCurve=DataInterface() {propertyValues {Curve {defaultValue=1, keys [<at 0.2 0 RCIM_Cubic>,<at 1 1 RCIM_Cubic>]}}}, Scale Curve=2)
    )
    Use Linear Drag=true
    Use Rotational Drag=true
   )
   &SolveForcesAndVelocity(Clamp Velocity=false, Limit Acceleration=false)
   &ScaleColor(
    Scale Alpha=&FloatFromCurve(
     FloatCurve=DataInterface() {propertyValues {Curve {defaultValue=1, keys [{interpMode=RCIM_Cubic, tangentMode=RCTM_User, time=0.75, value=1}, {tangentMode=RCTM_User, time=1, value=0}]}}}
     Scale Curve=1.5
    )
    Scale RGB=&VectorFromFloat(Value=&MakeCustomFloatFromBool(Bool=$User.bIsPainted, True Float=15, False Float=15))
    ScaleA=true
    ScaleRGB=true
   )
   &JitterPosition(Jitter Amount:NiagaraFloat=&RandomRangeFloat(Minimum=-4, Maximum=4), Jitter Delay=2)
  ]
  rendererClasses [&NiagaraSpriteRendererProperties]
  renderers [
   {rendererClass=&NiagaraSpriteRendererProperties}
   {rendererClass=&NiagaraSpriteRendererProperties, rendererIndex=1}
  ]
 }
 Sprites {
  simTarget=GPUComputeSim
  emitterSpawnScript [/SetVariables(Emitter.AcidDistortion:NiagaraFloat=&RandomRangeFloat(Minimum=0, Maximum=0.5)) {isSetParameters=true}]
  emitterUpdateScript [
   &EmitterState()
   &VC_FilteredSocketCount()
   &SpawnRate(
    SpawnRate=&MergeHelper_FloatSwitch(0=20, 1=40, 2=40, 3=?, Index=$System.GameModeID, Num=::NewEnumerator1)
    Use Spawn Probability=false
   )
   &SpawnRate(
    SpawnRate=&MergeHelper_FloatSwitch(0=0, 1=10, 2=10, 3=?, Index=$System.GameModeID, Num=::NewEnumerator1)
    Use Spawn Probability=false
   ) {module=SpawnRate002}
   &SpawnRate(
    SpawnGroup=1
    SpawnRate=&MakeCustomFloatFromBool(Bool=$User.bIsFrontendPreview, True Float=50, False Float=80)
    Use Spawn Probability=false
   ) {module=SpawnRate001}
  ]
  particleSpawnScript [
   &VC_PaintedVariantsOverride() {enabled=false}
   /SetVariables(
    Particles.Base Color:LinearColor=&MakeCustomLinearColorFromBool(Bool=$User.bIsPainted, True Color=$User.PaintedColor, False Color=<rgba 0.2 0.8 0.05 1>)
    Particles.VisibilityTag=&SpawnGroup()
   ) {isSetParameters=true}
   &InitializeParticle(
    Color Mode=::NewEnumerator0
    Lifetime Max=&MergeHelper_FloatSwitch(0=1, 1=0.6, 2=?, 3=?, Index=&SpawnGroup(), Num=::NewEnumerator0)
    Lifetime Min:NiagaraFloat=&MergeHelper_FloatSwitch(0=0.5, 1=0.5, 2=?, 3=?, Index=&SpawnGroup(), Num=::NewEnumerator0)
    Lifetime Mode=::NewEnumerator1
    Sprite Rotation Mode=::NewEnumerator2
    Sprite Size Mode=::NewEnumerator3
    Uniform Sprite Size=&MaskFloatBySpawnGroup(
     Masked Float=60
     Passthrough Float=&MaskFloatBySpawnGroup(
      Masked Float=&RandomRangeFloat(Minimum=90, Maximum=150)
      Passthrough Float=&MaskFloatBySpawnGroup(
       Masked Float=&RandomRangeFloat(
        Minimum=&MergeHelper_FloatSwitch(0=3, 1=20, 2=20, 3=?, Index=$System.GameModeID, Num=::NewEnumerator1)
        Maximum=&MergeHelper_FloatSwitch(0=6, 1=50, 2=50, 3=?, Index=$System.GameModeID, Num=::NewEnumerator1)
       )
       Passthrough Float=1
       Spawn Group Mask=2
      )
      Spawn Group Mask=1
     )
    )
    UsePositionOffset=false
    Write Material Random=true
   )
   &PartitionParticles(Number Of Partitions=$Emitter.PortCount)
   &SpriteFacingAndAlignment(Facing Coordinate Space=::Local, WriteAlignment=false, WriteFacing=true)
   &SkeletalMeshLocation(
    Bone / Socket Index=$Particles.PartitionParticles.Partition
    Bone / Socket Sampling Mode=::NewEnumerator5
    MaskSpawnGroup=false
    Mesh Sampling Type=::NewEnumerator6
    Sampled Position Offset=&AddVectorToPosition(
     Position=&MakeVector(
      X=&MergeHelper_FloatSwitch(0=-10, 1=-50, 2=-50, 3=?, Index=&SpawnGroup(), Num=::NewEnumerator1)
      Y=0
      Z=0
     )
     Vector=$User.DTCarOffset
    )
    Skeletal Mesh=$System.SkeletalMesh
   )
   &AddVelocity(Velocity=&MergeHelper_VectorSwitch(
     0=&MakeVector(
      X=&MaskFloatBySpawnGroup(
       Masked Float=-950
       Passthrough Float=&MaskFloatBySpawnGroup(
        Masked Float=-850
        Passthrough Float=&MaskFloatBySpawnGroup(Masked Float=-850, Passthrough Float=1, Spawn Group Mask=2)
        Spawn Group Mask=1
       )
      )
      Y=&MaskFloatBySpawnGroup(Masked Float=&RandomRangeFloat(Minimum=-80, Maximum=80), Passthrough Float=0, Spawn Group Mask=2)
      Z=&MaskFloatBySpawnGroup(Masked Float=&RandomRangeFloat(Minimum=-100, Maximum=100), Passthrough Float=0, Spawn Group Mask=2)
     )
     1=<0 0 0>
     2=<0 0 0>
     3=?
     Index=$System.GameModeID
     Num=::NewEnumerator1
    ))
   &Color(Color=&MergeHelper_ColorSwitch(0=<rgba 0.039 0.156 1.3 1>, 1=<rgba 0.05 0.2813 1 1>, 2=?, 3=?, Index=&SpawnGroup(), Num=::NewEnumerator0))
   &ShapeLocation(
    MaskSpawnGroup=false
    Override Seed=false
    Sphere Radius=&MaskFloatBySpawnGroup(Masked Float=20, Passthrough Float=0, Spawn Group Mask=2)
    UseOwnerScale=true
   )
  ]
  particleUpdateScript [
   &ParticleState()
   &ScaleSpriteSize(
    Scale Sprite Size Mode=::NewEnumerator0
    Uniform Scale Factor=&MergeHelper_FloatSwitch(
     0=&FloatFromCurve(FloatCurve=DataInterface() {propertyValues {Curve {keys [{arriveTangent=2, interpMode=RCIM_Cubic, time=0.3, value=1.15}, <at 1 2 RCIM_Cubic>]}}})
     1=1
     2=1
     3=?
     Index=&SpawnGroup()
     Num=::NewEnumerator1
    )
   )
   &SpriteRotationRate(Rotation Rate=&MaskFloatBySpawnGroup(Masked Float=&RandomRangeFloat(Minimum=-0.2, Maximum=0.2), Passthrough Float=0, Spawn Group Mask=1))
   &ScaleVelocity(
    Coordinate Space:ENiagaraCoordinateSpace=::Simulation
    Velocity Scale=&VectorFromFloat(Value=&MaskFloatBySpawnGroup(
      Masked Float=&MergeHelper_FloatSwitch(
       0=&FloatFromCurve(FloatCurve=DataInterface() {propertyValues {Curve {keys [<at 0 0.8>,<at 0.5 0.15>]}}})
       1=&FloatFromCurve(FloatCurve=DataInterface() {propertyValues {Curve {keys [<at 0 0.25>,<at 1 0>]}}})
       2=&FloatFromCurve(FloatCurve=DataInterface() {propertyValues {Curve {keys [<at 0 0.25>,<at 1 0>]}}})
       3=?
       Index=$System.GameModeID
       Num=::NewEnumerator1
      )
      Passthrough Float=&MaskFloatBySpawnGroup(
       Masked Float=&FloatFromCurve(FloatCurve=DataInterface() {propertyValues {Curve {keys [<at 0 1 RCIM_Cubic>,<at 1 1.25 RCIM_Cubic>]}}})
       Passthrough Float=&MaskFloatBySpawnGroup(Masked Float=0.6, Passthrough Float=1, Spawn Group Mask=2)
       Spawn Group Mask=1
      )
     ))
   )
   &SolveForcesAndVelocity(Clamp Velocity=false, Limit Acceleration=false)
   &ScaleColor(
    Scale Alpha=&MergeHelper_FloatSwitch(
     0=&FloatFromCurve(FloatCurve=DataInterface() {propertyValues {Curve {keys [<at -0.0046 1.0113 RCIM_Cubic>,<at 1.0011 0.002 RCIM_Cubic>]}}})
     1=&FloatFromCurve(FloatCurve=DataInterface() {propertyValues {Curve {keys [<at 0 1 RCIM_Cubic>,<at 1.0011 0.002 RCIM_Cubic>]}}})
     2=&FloatFromCurve(FloatCurve=DataInterface() {propertyValues {Curve {keys [<at -0.0046 1.0113 RCIM_Cubic>,<at 1.0011 0.002 RCIM_Cubic>]}}})
     3=?
     Index=&SpawnGroup()
     Num=::NewEnumerator1
    )
    Scale RGB=&VectorFromFloat(Value=&MergeHelper_FloatSwitch(
      0=&Multiply_Float(A=1, B=$User.PaintedScalar)
      1=$User.PaintedScalar
      2=&Multiply_Float(A=6, B=$User.PaintedScalar)
      3=?
      Index=&SpawnGroup()
      Num=::NewEnumerator1
     ))
    ScaleA=true
    ScaleRGB=true
   )
   &SubUVAnimation(
    End Frame Range Override=16
    Output Alternate Renderer Binding=true
    "Sprite Renderer "=DataInterface() {propertyValues {SpriteRenderer=&self:Sprites_0.NiagaraSpriteRendererProperties_0}}
    UseEndFrame=true
    UseStartFrame=false
   )
   &JitterPosition(Jitter Amount:NiagaraFloat=&MergeHelper_FloatSwitch(
     0=0
     1=&RandomRangeFloat(Minimum=-2, Maximum=2)
     2=&RandomRangeFloat(Minimum=-2, Maximum=2)
     3=?
     Index=&SpawnGroup()
     Num=::NewEnumerator1
    ))
  ]
  rendererClasses [&NiagaraSpriteRendererProperties]
  renderers [
   {rendererClass=&NiagaraSpriteRendererProperties}
   {rendererClass=&NiagaraSpriteRendererProperties, rendererIndex=1}
  ]
 }
 Trail {
  emitterUpdateScript [
   &EmitterState()
   &VC_FilteredSocketCount()
   &VC_RateQualityHandler(Epic Spawn Rate=30, High Spawn Rate=30, Low Spawn Rate=30, Use Custom Rate?=true)
   &VC_SpawnRibbon_v1(Number of Ribbons=$Emitter.PortCount, Spawn Time=-1, SpawnRate=$Emitter.Spawn Rate)
  ]
  particleSpawnScript [
   &VC_PaintedVariantsOverride() {enabled=false}
   &VC_InitializeRibbon_v1(
    Custom Offset=true
    Enable Skeletal Position=true
    Lifetime=0.8
    Local Vector Offset=&AddVectorToPosition(Position=<40 0 0>, Vector=$User.DTCarOffset)
    Offset Amount=1
    Ribbon Width=50
   )
   &AddVelocity(
    Velocity=&MakeCustomVectorFromBool(Bool=$User.bIsFrontendPreview, True Vector=<-1000 0 0>, False Vector=<-1500 0 0>)
    Velocity Speed Scale=&FloatFromCurve(
     FloatCurve=DataInterface() {propertyValues {Curve {keys [{interpMode=RCIM_Cubic, tangentMode=RCTM_User, time=0.5, value=1}, {interpMode=RCIM_Cubic, tangentMode=RCTM_User, time=1, value=0}]}}}
     Scale Curve=2
    )
   ) {module=AddVelocity001}
   /SetVariables(Particles.RibbonWidth2=35) {isSetParameters=true}
  ]
  particleUpdateScript [
   &ParticleState()
   &InheritSourceMovement(
    Applied Position Scale=&Multiply_VectorByFloat(
     Vector=<1 1 1>
     Float=&FloatFromCurve(FloatCurve=DataInterface() {propertyValues {Curve {keys [{interpMode=RCIM_Cubic, tangentMode=RCTM_Break, time=0.5, value=1}, {interpMode=RCIM_Cubic, tangentMode=RCTM_Break, time=1, value=0}]}}})
    )
    ApplyForce=false
    ApplyPosition=true
    ApplyPositionFalloff=false
   )
   &SolveForcesAndVelocity(Clamp Velocity=false, Limit Acceleration=false)
   &VC_UpdateRibbon_v1(Enable Skeletal Position=true)
   &Color(Color=&MergeHelper_ColorSwitch(
     0=&ColorFromCurve(ColorCurve=DataInterface() {propertyValues {
        AlphaCurve {keys [<at 0 1 RCIM_Cubic>,<at 1 0 RCIM_Cubic>]}
        BlueCurve {keys [
          <at 0 0.06>
          {arriveTangent=4.3632, interpMode=RCIM_Cubic, leaveTangent=4.3632, time=0.1554, value=0.03}
          {arriveTangent=0.8628, interpMode=RCIM_Cubic, leaveTangent=0.8628, time=0.2154, value=1}
          <at 0.9899 0.75>
         ]}
        GreenCurve {keys [
          <at 0 0.28>
          {arriveTangent=-0.836, interpMode=RCIM_Cubic, leaveTangent=-0.836, time=0.1554, value=0.14}
          {arriveTangent=-0.078, interpMode=RCIM_Cubic, leaveTangent=-0.078, time=0.2154, value=0.0999}
          <at 0.9899 0.0749>
         ]}
        RedCurve {keys [
          <at 0 1.62>
          {arriveTangent=-7.4744, interpMode=RCIM_Cubic, leaveTangent=-7.4744, time=0.1554, value=0.81}
          {arriveTangent=-0.9619, interpMode=RCIM_Cubic, leaveTangent=-0.9619, time=0.2154, value=0.0097}
          <at 0.9899 0.0073>
         ]}
       }})
     1=&ColorFromCurve(ColorCurve=DataInterface() {propertyValues {
        AlphaCurve {keys [<at 0 1 RCIM_Cubic>,<at 1 0 RCIM_Cubic>]}
        BlueCurve {keys [
          <at 0 0.06>
          {arriveTangent=4.3632, interpMode=RCIM_Cubic, leaveTangent=4.3632, time=0.1554, value=0.03}
          {arriveTangent=0.8628, interpMode=RCIM_Cubic, leaveTangent=0.8628, time=0.2154, value=1}
          <at 0.9899 0.75>
         ]}
        GreenCurve {keys [
          <at 0 0.28>
          {arriveTangent=-0.836, interpMode=RCIM_Cubic, leaveTangent=-0.836, time=0.1554, value=0.14}
          {arriveTangent=-0.078, interpMode=RCIM_Cubic, leaveTangent=-0.078, time=0.2154, value=0.0999}
          <at 0.9899 0.0749>
         ]}
        RedCurve {keys [
          <at 0 1.62>
          {arriveTangent=-7.4744, interpMode=RCIM_Cubic, leaveTangent=-7.4744, time=0.1554, value=0.81}
          {arriveTangent=-0.9619, interpMode=RCIM_Cubic, leaveTangent=-0.9619, time=0.2154, value=0.0097}
          <at 0.9899 0.0073>
         ]}
       }})
     2=&ColorFromCurve(ColorCurve=DataInterface() {propertyValues {
        AlphaCurve {keys [<at 0 1 RCIM_Cubic>,<at 1 0 RCIM_Cubic>]}
        BlueCurve {keys [
          <at 0 0.06>
          {arriveTangent=4.3632, interpMode=RCIM_Cubic, leaveTangent=4.3632, time=0.1554, value=0.03}
          {arriveTangent=0.8628, interpMode=RCIM_Cubic, leaveTangent=0.8628, time=0.2154, value=1}
          <at 0.9899 0.75>
         ]}
        GreenCurve {keys [
          <at 0 0.28>
          {arriveTangent=-0.836, interpMode=RCIM_Cubic, leaveTangent=-0.836, time=0.1554, value=0.14}
          {arriveTangent=-0.078, interpMode=RCIM_Cubic, leaveTangent=-0.078, time=0.2154, value=0.0999}
          <at 0.9899 0.0749>
         ]}
        RedCurve {keys [
          <at 0 1.62>
          {arriveTangent=-7.4744, interpMode=RCIM_Cubic, leaveTangent=-7.4744, time=0.1554, value=0.81}
          {arriveTangent=-0.9619, interpMode=RCIM_Cubic, leaveTangent=-0.9619, time=0.2154, value=0.0097}
          <at 0.9899 0.0073>
         ]}
       }})
     3=?
     Index=$System.GameModeID
     Num=::NewEnumerator1
    )) {module=Color001}
   /SetVariables(Particles.SecondTrailColor:LinearColor=&MergeHelper_ColorSwitch(
     0=&ColorFromCurve(ColorCurve=DataInterface() {propertyValues {
        AlphaCurve {keys [<at 0 1 RCIM_Cubic>,<at 1 0 RCIM_Cubic>]}
        BlueCurve {keys [
          <at 0 0.06>
          {arriveTangent=4.3632, interpMode=RCIM_Cubic, leaveTangent=4.3632, time=0.1554, value=0.03}
          {arriveTangent=1.1485, interpMode=RCIM_Cubic, leaveTangent=1.1485, time=0.2154, value=1}
          <at 1 1>
         ]}
        GreenCurve {keys [
          <at 0 0.28>
          {arriveTangent=-0.2441, interpMode=RCIM_Cubic, leaveTangent=-0.2441, time=0.1554, value=0.14}
          {arriveTangent=0.1358, interpMode=RCIM_Cubic, leaveTangent=0.1358, time=0.2154, value=0.2274}
          <at 1 0.2547>
         ]}
        RedCurve {keys [
          <at 0 1.62>
          {arriveTangent=-6.8232, interpMode=RCIM_Cubic, leaveTangent=-6.8232, time=0.1554, value=0.81}
          {arriveTangent=-0.746, interpMode=RCIM_Cubic, leaveTangent=-0.746, time=0.2154, value=0.15}
          <at 1 0.18>
         ]}
       }})
     1=&ColorFromCurve(ColorCurve=DataInterface() {propertyValues {
        AlphaCurve {keys [<at 0 1 RCIM_Cubic>,<at 1 0 RCIM_Cubic>]}
        BlueCurve {keys [
          <at 0 0.06>
          {arriveTangent=4.3632, interpMode=RCIM_Cubic, leaveTangent=4.3632, time=0.1554, value=0.03}
          {arriveTangent=1.1485, interpMode=RCIM_Cubic, leaveTangent=1.1485, time=0.2154, value=1}
          <at 1 1>
         ]}
        GreenCurve {keys [
          <at 0 0.28>
          {arriveTangent=-0.2441, interpMode=RCIM_Cubic, leaveTangent=-0.2441, time=0.1554, value=0.14}
          {arriveTangent=0.1358, interpMode=RCIM_Cubic, leaveTangent=0.1358, time=0.2154, value=0.2274}
          <at 1 0.2547>
         ]}
        RedCurve {keys [
          <at 0 1.62>
          {arriveTangent=-6.8232, interpMode=RCIM_Cubic, leaveTangent=-6.8232, time=0.1554, value=0.81}
          {arriveTangent=-0.746, interpMode=RCIM_Cubic, leaveTangent=-0.746, time=0.2154, value=0.15}
          <at 1 0.18>
         ]}
       }})
     2=&ColorFromCurve(ColorCurve=DataInterface() {propertyValues {
        AlphaCurve {keys [<at 0 1 RCIM_Cubic>,<at 1 0 RCIM_Cubic>]}
        BlueCurve {keys [
          <at 0 0.06>
          {arriveTangent=4.3632, interpMode=RCIM_Cubic, leaveTangent=4.3632, time=0.1554, value=0.03}
          {arriveTangent=1.1485, interpMode=RCIM_Cubic, leaveTangent=1.1485, time=0.2154, value=1}
          <at 1 1>
         ]}
        GreenCurve {keys [
          <at 0 0.28>
          {arriveTangent=-0.2441, interpMode=RCIM_Cubic, leaveTangent=-0.2441, time=0.1554, value=0.14}
          {arriveTangent=0.1358, interpMode=RCIM_Cubic, leaveTangent=0.1358, time=0.2154, value=0.2274}
          <at 1 0.2547>
         ]}
        RedCurve {keys [
          <at 0 1.62>
          {arriveTangent=-6.8232, interpMode=RCIM_Cubic, leaveTangent=-6.8232, time=0.1554, value=0.81}
          {arriveTangent=-0.746, interpMode=RCIM_Cubic, leaveTangent=-0.746, time=0.2154, value=0.15}
          <at 1 0.18>
         ]}
       }})
     3=?
     Index=$System.GameModeID
     Num=::NewEnumerator1
    )) {isSetParameters=true}
   &DynamicMaterialParameters(
    Index 0 Param 1=&FloatFromCurve(FloatCurve=DataInterface() {propertyValues {Curve {keys [<at 0 -0.1 RCIM_Cubic>,<at 1 -0.5 RCIM_Cubic>]}}})
    Index 0 Param 2=&RandomRangeFloat(Minimum=0, Maximum=1.5)
    Module.Index0Param0Write=true
    Module.Index0Param1Write=true
    Module.Index0Param2Write=true
    Module.Index0Param3Write=true
   )
   &ScaleRibbonWidth(Ribbon Width Scale=&FloatFromCurve(FloatCurve=DataInterface() {propertyValues {Curve {defaultValue=1, keys [{interpMode=RCIM_Cubic, tangentMode=RCTM_User, time=0.3, value=1}, {tangentMode=RCTM_User, time=1, value=0.4}]}}}))
  ]
  rendererClasses [&NiagaraRibbonRendererProperties]
  renderers [
   {rendererClass=&NiagaraRibbonRendererProperties}
   {rendererClass=&NiagaraRibbonRendererProperties, rendererIndex=1}
  ]
 }
}
summary {
 systemName=NS_Boost_AlienLiquidPSP_Optimized
 emitters [
  {emitterName=Trail, rendererClasses [&NiagaraRibbonRendererProperties]}
  {emitterName=Sprites, simTarget=GPUComputeSim, rendererClasses [&NiagaraSpriteRendererProperties]}
  {emitterName=NE_VC_Boost_Sprite_Parent, rendererClasses [&NiagaraSpriteRendererProperties]}
 ]
 userVariables [
  User.bIsFrontendPreview {type=NiagaraBool, defaultValue=true}
  User.bIsPainted {type=NiagaraBool}
  User.ColorVariantID {type=NiagaraInt32}
  User.DTCarOffset {type=Vector3f, description="", defaultValue {struct=Vector3f, value=<0 0 0>}}
  User.GameModeID {type=NiagaraInt32}
  User.PaintedColor {type=LinearColor, description="", defaultValue {struct=LinearColor, value=<rgba 1 1 1 1>}}
  User.PaintedScalar {type=NiagaraFloat, defaultValue=1, description=""}
  User.PortCount {type=NiagaraInt32, defaultValue=2}
 ]
}
